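; Temporary fake-light Blueprint, named DoNotUse, that shows or hides its Cube component based on the player's Visual Effect Quality setting.
; ScalabilityBasedVisibility selects one of five per-tier variables (Low through Cinematic) and applies it without propagating to child components.
; It runs in the construction script, on BeginPlay and through an Update Scalability event bound to scalability settings.
; Tick only calls the parent, and the overlap event is empty.
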(event EventBeginPlay
  (|CallParentFunction)
  (CallFunction|ScalabilityBasedVisibility)
  (Default|BindScalabilitySettings (delegate Custom|Update%20Scalability)))

(event Collision|EventActorBeginOverlap
  (params (OtherActor :type Actor%20Object%20Reference)))

(event Custom|Update%20Scalability
  (CallFunction|ScalabilityBasedVisibility))

(event EventTick
  (params (DeltaSeconds :type Float%20%28single-precision%29))
  (|CallParentFunction DeltaSeconds))

(fn ScalabilityBasedVisibility
  (bind _returnvalue (Settings|GetGameUserSettings))
  (Rendering|SetVisibility (Variables|Default|GetCube) (Utilities|Select :Index (Settings|GetVisualEffectQuality _returnvalue) :Option%200 (Variables|4-Scalability|GetLow) :Option%201 (Variables|4-Scalability|GetMedium) :Option%202 (Variables|4-Scalability|GetHigh) :Option%203 (Variables|4-Scalability|GetEpic) :Option%204 (Variables|4-Scalability|GetCinematic%20)) false))

(fn ConstructionScript
  (|CallParentFunction)
  (CallFunction|ScalabilityBasedVisibility))
